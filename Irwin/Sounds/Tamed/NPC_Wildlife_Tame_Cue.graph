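# Sound cue for the tamed-wildlife NPC taming sound, in the creatures sound class.
# It plays a single wave, Universal_Tame_02, through one wave player with no randomization or modulation.
# A dedicated tame attenuation asset shapes falloff.
# A concurrency override caps it at two simultaneous instances.
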SoundCue /Irwin/Sounds/Tamed/NPC_Wildlife_Tame_Cue
root=WavePlayer_2
nodes [{id=WavePlayer_2, properties {soundWaveAssetPtr=&/Universal_Tame_02}}]
properties {
 attenuationSettings=&/NPC_Wildlife_Tame_Att
 bOverrideConcurrency=true
 maxDistance=2000
 soundClassObject=/Game/Sounds/FortSoundClassesAndModes/Fort_Creatures
 volumeMultiplier=1
 concurrencyOverrides {maxCount=2}
}
